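# Unlit additive material for Niagara mesh particles, built for the button effect on the Detector backbling gadget.
# Particle color feeds a shared emissive adjustment function.
# Particle alpha scales a squared product of the second UV channel's U and a normal-to-camera facing term, forming a soft, view-dependent fade.
# The normalized camera direction, scaled by a Camera Offset parameter, provides a vector offset toward the viewer.
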Material /Game/Animation/Props/Gadgets_Backbling_Detector/FX/Materials/M_Backbling_Detector_Button
graph {
 properties {
  bUsedWithNiagaraMeshParticles=true
  blendMode=BLEND_Additive
  pixelDepthOffsetMode=PDOM_Legacy
  previewMesh=/Game/Animation/Props/Gadgets_Backbling_Detector/FX/Meshes/SM_Backbling_Detector_Button
  shadingModel=MSM_Unlit
  usedShadingModels=Unlit
 }
 nodes [
  n1 = ParticleColor()
  MaterialFunctionCall(Emissive=@n1.RGB) {materialFunction=&MF_FX_EmissiveAdjustments}
  n2 = Normalize(VectorInput=Subtract(A=CameraPositionWS(), B=WorldPosition().XYZ))
  n3 = Multiply(A=ComponentMask(None=TextureCoordinate() {coordinateIndex=1}) {r=true}, B=DotProduct(A=VertexNormalWS(), B=@n2))
  Multiply(A=@n1.A, B=Multiply(A=@n3, B=@n3))
  Multiply(A=@n2, B=ScalarParameter() {defaultValue=0.5, parameterName=Camera Offset})
 ]
}
