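# Master character material for the Detector gadget back bling, animating its dial in the shader through World Position Offset.
# A UV-box mask limits a pre-skinned local-space rotation to the needle, swung between Dial Max and Dial Min by the Dial parameter.
# Sine jitter grows sharply as Dial nears 1, while the default negative Dial switches the offset off.
# The offset is injected between the standard character PreFX, SkinFX and PostFX functions.
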Material /Game/Animation/Props/Gadgets_Backbling_Detector/Materials/M_Gadget_Backbling_Detector_MASTER
graph {
 properties {
  bAutomaticallySetUsageInEditor=false
  bUseMaterialAttributes=true
  bUsedWithClothing=true
  bUsedWithMeshDeformer=true
  bUsedWithMorphTargets=true
  bUsedWithSkeletalMesh=true
  pixelDepthOffsetMode=PDOM_Legacy
  refractionMethod=RM_IndexOfRefraction
  shadingModel=MSM_FromMaterialExpression
  subsurfaceProfile=&SS_HeroSkin_02
  usedShadingModels=Default Lit | Subsurface | Clear Coat | Subsurface Profile | Cloth | Eye
 }
 nodes [
  n1 = MaterialFunctionCall() {materialFunction=&BaseFN_CharacterFunction_PreFX}
  n2 = ScalarParameter() {defaultValue=-1, parameterName=Dial}
  n3 = Saturate(None=Round(None=Add(A=@n2)))
  n4 = Time()
  n5 = Multiply(A=LinearInterpolate(A=Sine(None=@n4) {period=0.2}, B=Sine(None=@n4) {period=0.15}, Alpha=ScalarParameter() {defaultValue=1, parameterName=Dial Close})) {constB=0.025}
  n6 = Power(Base=Saturate(None=@n2)) {constExponent=4}
  n7 = LinearInterpolate(
   A=ScalarParameter() {defaultValue=0.34, parameterName=Dial Max}
   B=Add(A=ScalarParameter() {defaultValue=-0.105, parameterName=Dial Min}, B=Multiply(A=@n5, B=@n6))
   Alpha=@n2
  ) {constA=0.15, constB=-0.1}
  n8 = TextureCoordinate()
  n9 = Constant2Vector() {g=0.135, r=0.595}
  n10 = Constant2Vector() {g=0.059, r=0.567}
  n11 = Constant2Vector()
  n12 = If(
   A=Divide(A=Subtract(A=@n8, B=@n9), B=Subtract(A=@n10, B=@n9))
   B=Constant2Vector() {g=0.5, r=0.5}
   "A > B"=@n11
   "A == B"=Constant2Vector() {g=1, r=1}
   "A < B"=@n11
  ) {equalsThreshold=0.5}
  n13 = Multiply(A=ComponentMask(None=@n12) {r=true}, B=ComponentMask(None=@n12) {g=true})
  n14 = Transform(None=RotateAboutAxis(
    NormalizedRotationAxis=Constant3Vector() {constant=<rgba 0 1 0 0>}
    RotationAngle=Multiply(A=@n7, B=@n13)
    PivotPoint=Constant3Vector() {constant=<rgba 0 0 -8 1>}
    Position=PreSkinnedPosition()
   )) {transformSourceType=TRANSFORMSOURCE_Local}
  n15 = MaterialFunctionCall(
   AO=@n1.AO
   Distance Blend=@n1.AO
   Fresnel=@n1.AO
   Gmap=@n1.AO
   Pre Skinned Local Position=@n1.AO
   Material=SetMaterialAttributes!(
    MaterialAttributes=MaterialFunctionCall(
     PreskinnedPosition=@n1.Pre skinned Local Position
     Gmap=@n1.Pre skinned Local Position
     MaterialAttributes=@n1.Pre skinned Local Position
     AO=@n1.Pre skinned Local Position
    ) {materialFunction=&CharacterSkinFX}.Result
    World Position Offset=Multiply(A=@n3, B=@n14)
   ) {attributeSetTypes [#1]}
  ) {materialFunction=&BaseFN_CharacterFunction_PostFX}
  MaterialFunctionCall(In=@n15.Result) {materialFunction=&CINE_MF_CharPostFX}
 ]
}
